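# Skydiving contrail Niagara system for the CozyMechanic cosmetic, a saved previous iteration, with spinning gears, glows, lights, sparks, nuts and bolts, and trails.
# Emitters attach to skeletal mesh bones, with gear and spark offsets blended by System.Audrey and System.Raph.
# SparkPosition particles supply position and direction to Griding_Sparks through a particle reader.
# Paraglide and lean alphas vary spin, speed, size and spark behavior; debug and gizmo emitters are disabled.
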NiagaraSystem /BRCosmetics/Athena/Effects/Particles/Skydiving/NS_Contrail_CozyMechanic/Previous_Iterations/NS_CozyMechanic_Contrail_Save
dependencies {
 usedDataInterfaces [
  &NiagaraDataInterfaceArrayFloat
  &NiagaraDataInterfaceCurve
  &NiagaraDataInterfaceColorCurve
  &NiagaraDataInterfaceArrayFloat3
  &NiagaraDataInterfaceArrayInt32
  &NiagaraDataInterfaceParticleRead
  &NiagaraDataInterfaceSkeletalMesh
  &NiagaraDataInterfaceCamera
 ]
 usedDynamicInputs [
  &MakeFloatFromVector
  &TransformVector
  &FNTimeOfDayScalar
  &RemapRange
  &DotProduct
  &ReturnCameraVector
 ]
 usedModules [
  &self:SystemSpawnScript.SystemScriptSource.SystemScriptGraph.NiagaraNodeAssignment_0.SetVariables_#1
  &SystemState
  &self:UpdatePreviousRotation
  &self:SystemSpawnScript.SystemScriptSource.SystemScriptGraph.NiagaraNodeAssignment_5.SetVariables_#2
  &self:SystemSpawnScript.SystemScriptSource.SystemScriptGraph.NiagaraNodeAssignment_6.SetVariables_#3
  &self:SystemSpawnScript.SystemScriptSource.SystemScriptGraph.NiagaraNodeAssignment_1.SetVariables_#4
  &self:SystemSpawnScript.SystemScriptSource.SystemScriptGraph.NiagaraNodeAssignment_7.SetVariables_#5
  &FN_Integrator
  &self:SystemSpawnScript.SystemScriptSource.SystemScriptGraph.NiagaraNodeAssignment_4.SetVariables_#6
  &self:SystemSpawnScript.SystemScriptSource.SystemScriptGraph.NiagaraNodeAssignment_8.SetVariables_#7
  /Game/Effects/Niagara/Modules/Contrails/FN_Contrail_Setup
  &self:SystemSpawnScript.SystemScriptSource.SystemScriptGraph.NiagaraNodeAssignment_2.SetVariables_#8
 ]
}
emitters {
 FN_Debug {
  bEnabled=false
  emitterUpdateScript [
   &EmitterState()
   &SpawnBurst_Instantaneous(Spawn Count=1, Use Spawn Probability=false)
  ]
  particleSpawnScript [/Niagara/Modules/Spawn/Initialization/InitializeParticle(
    Lifetime=5
    Module.Write Color=true
    Module.Write Lifetime=true
    Module.Write Mass=true
    Module.Write Material Random=true
    Module.Write Position=false
    Module.Write Scale=false
    Module.Write SpriteRotation=false
    Module.Write SpriteSize=true
    Position=&SimulationPosition()
    Sprite Size=<100 100>
   )]
  particleUpdateScript [
   &ParticleState()
   /SetVariables(Particles.Debug:NiagaraFloat=$User.ParaGlideLeanAlpha) {isSetParameters=true}
   &DynamicMaterialParameters(
    Index 0 Param 1=$Particles.Debug
    Module.Param0WriteEnabled=true
    Module.Param1WriteEnabled=true
    Module.Param2WriteEnabled=true
    Module.Param3WriteEnabled=true
   )
  ]
  rendererClasses [&NiagaraSpriteRendererProperties]
  renderers [{rendererClass=&NiagaraSpriteRendererProperties}]
 }
 Gear {
  emitterUpdateScript [
   &EmitterState(Life Cycle Mode=::NewEnumerator1, Loop Behavior=::NewEnumerator1, Loop Duration Mode=::NewEnumerator1, UseLoopDelay=false)
   &SpawnBurst_Instantaneous(Spawn Count=2, Use Spawn Probability=false, UseLoopCountLimit=false)
   &SpawnBurst_Instantaneous(Spawn Count=2, Spawn Group=1, Use Spawn Probability=false, UseLoopCountLimit=false) {module=SpawnBurst_Instantaneous001}
   &SpawnBurst_Instantaneous(Spawn Count=2, Spawn Group=2, Use Spawn Probability=false, UseLoopCountLimit=false) {module=SpawnBurst_Instantaneous002}
   &SpawnBurst_Instantaneous(Spawn Count=2, Spawn Group=3, Use Spawn Probability=false, UseLoopCountLimit=false) {module=SpawnBurst_Instantaneous003}
  ]
  particleSpawnScript [
   /Niagara/Modules/Spawn/Initialization/V2/InitializeParticle(Lifetime=5, Mesh Scale Mode=::NewEnumerator3, UsePositionOffset=false, Write Material Random=true)
   &PartitionParticles()
   &InitialMeshOrientation(
    Mesh Orientation Mode=::NewEnumerator2
    Orientation Vector=<0 0 1>
    Rotation=&SelectVectorFromArray(
     Direct Array Index=$Particles.PartitionParticles.Partition
     Vector Selection Array=DataInterface() {propertyValues {FloatData [<0 0 -0.25>, <0 0 0.25>]}}
    )
    Rotation Coordinate Space=::NewEnumerator2
    Use Rotation Vector=true
   )
   &ApplyInitialForces()
   /SetVariables() {isSetParameters=true, module=SetVariables_#9}
   /SetVariables(Particles.VisibilityTag=&SpawnGroup()) {isSetParameters=true, module=SetVariables_#10}
  ]
  particleUpdateScript [
   /SetVariables(Particles.TimeTracker:NiagaraFloat=&Multiply_Float(A=&Add_Float(A=$Engine.DeltaTime, B=$Particles.TimeTracker), B=1)) {isSetParameters=true, module=SetVariables_#11}
   &ParticleState(Kill Particles When Lifetime Has Elapsed=false)
   &ScaleMeshSize(Scale Factor=&VectorFromFloat(Value=&FloatFromCurve(FloatCurve=DataInterface() {propertyValues {Curve {defaultValue=1, keys [{tangentMode=RCTM_User, time=0, value=0}, {tangentMode=RCTM_User, time=0.0726, value=1.2439}, <at 0.1324 1>]}}})))
   &SkeletalMeshLocation(
    Bone / Socket Index=$Particles.PartitionParticles.Partition
    Bone Sampling Mode=::NewEnumerator4
    MaskSpawnGroup=false
    Orientation Sampling=::NewEnumerator1
    Sampled Position Offset=&MergeHelper_VectorSwitch(
     0=<0 0 0>
     1=&Lerp_Vector(
      A=&Lerp_Vector(A=<0 20 12>, B=<0 0 25>, Alpha=$System.Audrey)
      B=&Lerp_Vector(A=<0 -20 -12>, B=<0 0 -25>, Alpha=$System.Raph)
      Alpha=&MakeFloatFromInt(Int=$Particles.PartitionParticles.Partition)
     )
     2=&Lerp_Vector(
      A=&Lerp_Vector(A=<0 20 -10>, B=<0 17 15>, Alpha=$System.Audrey)
      B=&Lerp_Vector(A=<0 -20 10>, B=<0 -17 -15>, Alpha=$System.Raph)
      Alpha=&MakeFloatFromInt(Int=$Particles.PartitionParticles.Partition)
     )
     3=&Lerp_Vector(
      A=&Lerp_Vector(A=<0 36 0>, B=<0 16 35>, Alpha=$System.Audrey)
      B=&Lerp_Vector(A=<0 -36 0>, B=<0 -16 -35>, Alpha=$System.Raph)
      Alpha=&MakeFloatFromInt(Int=$Particles.PartitionParticles.Partition)
     )
     Index=&SpawnGroup()
     Num=::NewEnumerator2
    )
    Skeletal Mesh=$System.Skeletal Mesh
   )
   /SetVariables(Particles.MeshOrientation=&self:ScratchDynamicInput(
     SampledRotation:Quat4f=$Particles.SkeletalMeshLocation.SampledRotation
     InitialRotation:Quat4f=$Particles.Initial.MeshOrientation
     OffsetQuaternion:Quat4f=/BRCosmetics/Athena/Effects/Particles/Skydiving/NS_Contrail_Apprentice/OrientMestToVectorFromCustomRot(
      InitialRotation:Quat4f=<0 0 0 1>
      Look At Direction:Vector3f=<0 0 0>
      Reference Axis Vector:Vector3f=<0 0 0>
      RotationVector:Vector3f=&MakeVector(
       X=0
       Y=&Multiply_Float(
        A=$Particles.TimeTracker
        B=&Multiply_Float(
         A=&MergeHelper_FloatSwitch(0=0.4, 1=-0.4, 2=-0.4, 3=0.4, Index=&SpawnGroup(), Num=::NewEnumerator2)
         B=&Lerp_Float(A=0.8, B=3, Alpha=$System.ParaGlideAlphaFloat)
        )
       )
       Z=0
      )
     )
    )) {isSetParameters=true}
   &SolveForcesAndVelocity(Clamp Velocity=false, Limit Acceleration=false)
   &Color(Color=&ColorFromCurve(ColorCurve=DataInterface() {propertyValues {
       AlphaCurve {keys [<at 0 1 RCIM_Cubic>,<at 1 1 RCIM_Cubic>]}
       BlueCurve {keys [<at 0 1 RCIM_Cubic>,<at 0.2322 0.3342>,<at 0.3832 0.2244>]}
       GreenCurve {keys [<at 0 1 RCIM_Cubic>,<at 0.2322 0.4181>,<at 0.3832 0.3159>]}
       RedCurve {keys [<at 0 1 RCIM_Cubic>,<at 0.2322 0.9167>,<at 0.3832 0.9167>]}
      }}))
  ]
  rendererClasses [&NiagaraMeshRendererProperties]
  renderers [
   {rendererClass=&NiagaraMeshRendererProperties, rendererIndex=0}
   {rendererClass=&NiagaraMeshRendererProperties, rendererIndex=1}
   {rendererClass=&NiagaraMeshRendererProperties, rendererIndex=2}
   {rendererClass=&NiagaraMeshRendererProperties, rendererIndex=3}
  ]
 }
 Gizmo_LocalSpace {
  bEnabled=false
  emitterUpdateScript [
   &EmitterState()
   &SpawnBurst_Instantaneous(Spawn Count=4, Spawn Probability=HlslExpression() {hlslExpression="Emitter.LoopCount == 0"}, Use Spawn Probability=true)
  ]
  particleSpawnScript [
   &InitializeParticle(
    Color=<rgba 0 0.3677 1 1>
    Lifetime Max=1
    Lifetime Mode=::NewEnumerator1
    Mesh Scale Mode=::NewEnumerator3
    Mesh Uniform Scale=3
    UsePositionOffset=false
    Write Material Random=true
   )
   &InitialMeshOrientation(Mesh Axis To Orient=<0 0 -1>, Mesh Orientation Mode=::NewEnumerator2, Use Rotation Vector=false)
  ]
  particleUpdateScript [
   &ParticleState(Kill Particles When Lifetime Has Elapsed=false, Let Infinitely Lived Particles Die When Emitter Deactivates=true)
   &UpdateMeshOrientation(
    Facing Direction=$System.FN_Contrail_Setup.Direction Vector
    Facing Interpolation=false
    Orientation Method=::NewEnumerator1
    Override Quat=false
    Stabilization Mode:ENiagara_OrientVectorCount=::NewEnumerator2
    TransformFacing=false
    TransformSide=false
   )
   /SetVariables(Particles.Position=&AddVectorToPosition(
     Position=$Particles.Initial.Position
     Vector=&Multiply_VectorByFloat(
      Vector=$System.FN_Contrail_Setup.Direction Velocity
      Float=&FloatFromCurve(
       FloatCurve=DataInterface() {propertyValues {Curve {keys [{tangentMode=RCTM_User, time=0, value=0}, {tangentMode=RCTM_User, time=1, value=1}]}}}
       Scale Curve=$Particles.Initial.Lifetime
      )
     )
    )) {enabled=false, isSetParameters=true}
   /SetVariables(Particles.Velocity=$System.FN_Contrail_Setup.Direction Vector) {isSetParameters=true}
  ]
  rendererClasses [&NiagaraMeshRendererProperties]
  renderers [{rendererClass=&NiagaraMeshRendererProperties, rendererIndex=0}]
 }
 GlowAndLight {
  emitterUpdateScript [
   &EmitterState(
    Inactive Response=::NewEnumerator1
    Life Cycle Mode=::NewEnumerator1
    Loop Behavior=::NewEnumerator1
    Loop Duration Mode=::NewEnumerator1
    UseLoopDelay=false
   )
   &SpawnBurst_Instantaneous(Spawn Count=4, Use Spawn Probability=false, UseLoopCountLimit=false)
   &SpawnBurst_Instantaneous(Spawn Count=2, Spawn Group=1, Use Spawn Probability=false, UseLoopCountLimit=false) {module=SpawnBurst_Instantaneous001}
  ]
  particleSpawnScript [
   &PartitionParticles(Number Of Partitions=4)
   &InitializeParticle(
    Color=<rgba 10 2.5 0 0.2>
    Lifetime=5
    Mesh Scale Mode=::NewEnumerator3
    Sprite Rotation Mode=::NewEnumerator1
    Sprite Size Mode=::NewEnumerator3
    Uniform Sprite Size=&MergeHelper_FloatSwitch(0=180, 1=140, 2=80, 3=80, Index=$Particles.PartitionParticles.Partition, Num=::NewEnumerator2)
    UsePositionOffset=false
    Write Material Random=true
   )
   /SetVariables() {isSetParameters=true}
   /SetVariables(Particles.VisibilityTag=&SpawnGroup()) {isSetParameters=true}
   &CameraOffset() {module=CameraOffset001}
  ]
  particleUpdateScript [
   &ParticleState(Kill Particles When Lifetime Has Elapsed=false, Let Infinitely Lived Particles Die When Emitter Deactivates=true)
   &SkeletalMeshLocation(
    Bone / Socket Index=$Particles.PartitionParticles.Partition
    Bone Sampling Mode=::NewEnumerator4
    MaskSpawnGroup=false
    Orientation Sampling=::NewEnumerator1
    Sampled Position Offset=&MergeHelper_VectorSwitch(0=<15 0 8>, 1=<-15 0 -8>, 2=<-8 0 0>, 3=<8 0 0>, Index=$Particles.PartitionParticles.Partition, Num=::NewEnumerator2)
    Skeletal Mesh=$System.Skeletal Mesh
   ) {module=SkeletalMeshLocation001}
   /SetVariables(Particles.SpriteAlignment=&NormalizeVector(
     Vector To Normalize=&Add_Vector(
      A=&Add_Vector(
       A=$Particles.SpriteAlignment
       B=&NormalizeVector(
        Vector To Normalize=&Multiply_Vector(A=$Engine.Owner.Velocity, B=<1 1 -1>)
        Safe Normalize=true
        Safe Fallback Threshold=0
        Safe Fallback Vector=<0 0 0.001>
       )
      )
      B=<0 0 0.4>
     )
     Safe Normalize=true
     Safe Fallback Threshold=0
     Safe Fallback Vector=<0 0 0.001>
    )) {isSetParameters=true}
   &DynamicMaterialParameters(
    Index 0 Param 3=1
    Module.Index0Param0Write=true
    Module.Index0Param1Write=true
    Module.Index0Param2Write=true
    Module.Index0Param3Write=true
    Module.Index1Param0Write=true
    Module.Index1Param1Write=true
    Module.Index1Param2Write=true
    Module.Index1Param3Write=true
    Write Parameter Index 1=true
   )
   &SolveForcesAndVelocity(Clamp Velocity=false, Limit Acceleration=false)
   &FN_Light_Attributes(
    Color=<rgba 1 0.2885 0.0513 1>
    Create Color Attribute=true
    Create Exponent Attribute=true
    Create Position Attribute=true
    Create Radius Attribute=true
    Exponent=5
    Position=&AddVectorToPosition(
     Position=$StackContext.SkeletalMeshLocation001.SampledPosition
     Vector=&MakeVector(
      X=&MergeHelper_FloatSwitch(0=10, 1=-10, 2=?, 3=?, Index=$Particles.PartitionParticles.Partition, Num=::NewEnumerator0)
      Y=0
      Z=20
     )
    )
    Radius=60
    Scale Color=&Multiply_Float(A=25, B=$System.RevealCurve)
   )
   &CameraOffset(Camera Offset Amount=5)
   /SetVariables(Particles.Vector:Vector3f=<1 1 1>) {isSetParameters=true}
   &ScaleColor(
    Scale Alpha=&Multiply_Float(
     A=&Multiply_Float(A=$System.RevealCurve, B=$System.ToD)
     B=&FN_FrontEndFloatModifier(Gameplay=0.8, Front End=0.5)
    )
    ScaleA=true
    ScaleRGB=true
   )
   &ScaleSpriteSize(
    Scale Sprite Size Mode=::NewEnumerator0
    Uniform Scale Factor=&Lerp_Float(
     A=1
     B=&MergeHelper_FloatSwitch(0=1.5, 1=1.5, 2=1, 3=1, Index=$Particles.PartitionParticles.Partition, Num=::NewEnumerator2)
     Alpha=$System.ParaGlideAlphaFloat
    )
   )
   /SetVariables(
    Particles.Linear Color=&ColorFromCurve(ColorCurve=DataInterface() {propertyValues {
       AlphaCurve {keys [<at 0 0 RCIM_Cubic>,<at 0.3911 0.2>,<at 1 0.15 RCIM_Cubic>]}
       BlueCurve {keys [<at 0 0 RCIM_Cubic>,<at 1 0 RCIM_Cubic>]}
       GreenCurve {keys [<at 0 0.6793 RCIM_Cubic>,<at 1 0.3368 RCIM_Cubic>]}
       RedCurve {keys [<at 0 1 RCIM_Cubic>,<at 1 1 RCIM_Cubic>]}
      }})
    Particles.Vector 2D:Vector2f=&Multiply_Vector2D(
     A=$Particles.Initial.SpriteSize
     B=&Vector2DFromFloat(Value=&MergeHelper_FloatSwitch(0=1, 1=1, 2=0, 3=0, Index=$Particles.PartitionParticles.Partition, Num=::NewEnumerator2))
    )
   ) {enabled=false, isSetParameters=true}
  ]
  rendererClasses [&NiagaraLightRendererProperties, &NiagaraSpriteRendererProperties]
  renderers [
   {rendererClass=&NiagaraLightRendererProperties, rendererIndex=0}
   {rendererClass=&NiagaraSpriteRendererProperties, rendererIndex=1}
   {rendererClass=&NiagaraSpriteRendererProperties, rendererIndex=2}
  ]
 }
 Griding_Sparks {
  emitterUpdateScript [
   &EmitterState(Life Cycle Mode=::NewEnumerator1, Loop Behavior=::NewEnumerator1, Loop Delay=0.25, Loop Duration Mode=::NewEnumerator1, UseLoopDelay=true)
   &SpawnParticlesFromOtherEmitter(
    Attribute Reader=DataInterface() {propertyValues {EmitterBinding {bindingMode=Other, emitterName=SparkPosition}}}
    Spawn Rate=&Lerp_Float(A=120, B=120, Alpha=$System.ParaGlideAlphaFloat)
    Spawn Rate Per Particle Cap=1500
    Use Maximum Age=false
    Use Minimum Age=false
   )
  ]
  particleSpawnScript [
   &InitializeParticle(
    Color=<rgba 50 50 50 1>
    Lifetime=&Lerp_Float(
     A=&RandomRangeFloat(Minimum=0.1, Maximum=0.6)
     B=&RandomRangeFloat(Minimum=0.4, Maximum=0.6)
     Alpha=$User.ParaGlideLeanAlpha
    )
    Mesh Scale=/Niagara/DynamicInputs/CascadeConversion/SplitVector(
     Double Component Vector Channels:ECascadeNiagaraTwoVectorChannels=::NewEnumerator0
     Double Component:NiagaraFloat=&RandomRangeFloat(Minimum=0.15, Maximum=0.2)
     Single Component:NiagaraFloat=&RandomRangeFloat(Minimum=0.5, Maximum=0.8)
    )
    Mesh Scale Mode=::NewEnumerator4
    Sprite Size=&RandomRangeVector2D(Minimum=<0.5 3>, Maximum=<0.5 4>)
    Sprite Size Mode=::NewEnumerator4
    UsePositionOffset=false
    Write Material Random=true
   )
   &SampleParticlesFromOtherEmitter(Applied Velocity Scale=20, Velocity Application:ENiagara_DirectReadApplicationMode=::NewEnumerator0, Velocity Sampling=::NewEnumerator0)
   &AddVelocity(
    Cone Angle=&Lerp_Float(A=120, B=130, Alpha=$System.ParaGlideAlphaFloat)
    Cone Axis=$Particles.Velocity
    Module.UseConeFalloff=false
    Module.UseVelDistribution=false
    Override Seed=false
    Velocity Mode=::NewEnumerator2
    Velocity Speed=&Lerp_Float(
     A=&RandomRangeFloat(Minimum=30, Maximum=150)
     B=&RandomRangeFloat(Minimum=30, Maximum=150)
     Alpha=$System.ParaGlideAlphaFloat
    )
   )
  ]
  particleUpdateScript [
   &ParticleState()
   &ScaleSpriteSize(Uniform Curve Sprite Scale=DataInterface() {propertyValues {Curve {defaultValue=0, keys [{tangentMode=RCTM_User, time=0, value=1}, {tangentMode=RCTM_User, time=1, value=0}]}}}) {enabled=false}
   /SetVariables(Particles.SpriteSize=&Multiply_Vector2D_ByFloat(
     Float=&Lerp_Float(
      A=&FloatFromCurve(FloatCurve=DataInterface() {propertyValues {Curve {keys [<at 0 1>,<at 0.3042 0.4618>,<at 1 0>]}}})
      B=&FloatFromCurve(FloatCurve=DataInterface() {propertyValues {Curve {keys [<at 0 1>,<at 0.253 0.4411>,<at 1 0.045>]}}})
      Alpha=$System.ParaGlideAlphaFloat
     )
     Vector2D=$Particles.Initial.SpriteSize
    )) {isSetParameters=true}
   /Niagara/Modules/Update/Forces/CurlNoiseForce(
    Noise Frequency=100
    Noise Strength=&Multiply_Float(A=2000, B=&Lerp_Float(A=0.3, B=1, Alpha=&ClampFloat(Float=$User.ParaGlideLeanAlpha)))
    Pan Noise=true
   ) {enabled=false}
   &Drag(Drag=2.5, Use Linear Drag=true, Use Rotational Drag=true)
   &SolveForcesAndVelocity(Clamp Velocity=false, Limit Acceleration=false)
   /SetVariables(Particles.Glow03_Color=&MakeLinearColorFromVectorAndFloat(
     "Vector (RGB)"=&MakeVectorFromLinearColorRGB(Color=&MultiplyLinearColor(Linear Color=<rgba 1.5 0.05 0 1>, Scale Factor=&MakeVector4_FromFloat(Float=1)))
     "Float (Alpha)"=&Multiply_Float(
      A=0.04
      B=&FloatFromCurve(
       FloatCurve=DataInterface() {propertyValues {Curve {defaultValue=0, keys [{tangentMode=RCTM_User, time=0, value=1}, {tangentMode=RCTM_User, time=1, value=0}]}}}
       Scale Curve=0.5
      )
     )
    )) {isSetParameters=true, module=SetVariables_#12}
   &ScaleColor(
    Scale Alpha=&FloatFromCurve(FloatCurve=DataInterface() {propertyValues {Curve {keys [<at 0 0>,<at 0.2 1>,<at 0.4 1>,<at 1 1>]}}})
    Scale RGB=&MakeVectorFromLinearColorRGB(Color=&ColorFromCurve(ColorCurve=DataInterface() {propertyValues {
        AlphaCurve {keys [<at 0 1 RCIM_Cubic>,<at 1 0 RCIM_Cubic>]}
        BlueCurve {keys [<at 0.1638 0 RCIM_Cubic>,<at 0.498 0 RCIM_Cubic>]}
        GreenCurve {keys [<at 0.1638 1.4 RCIM_Cubic>,<at 0.498 0.05 RCIM_Cubic>]}
        RedCurve {keys [<at 0.1638 3 RCIM_Cubic>,<at 0.498 1 RCIM_Cubic>]}
       }}))
    ScaleA=false
    ScaleRGB=true
   )
   /SetVariables(Particles.Glow03_Size=&Multiply_Vector2D_ByFloat(
     Float=&FloatFromCurve(
      FloatCurve=DataInterface() {propertyValues {Curve {keys [
          <at 0 0 RCIM_Cubic>
          <at 0.2 1.3 RCIM_Cubic>
          {arriveTangent=-0.1423, interpMode=RCIM_Cubic, leaveTangent=-0.1423, tangentMode=RCTM_User, time=1, value=0}
         ]}}}
      Scale Curve=10
     )
     Vector2D=<0.6 1.2>
    )) {isSetParameters=true, module=SetVariables_#13}
  ]
  rendererClasses [&NiagaraSpriteRendererProperties]
  renderers [
   {rendererClass=&NiagaraSpriteRendererProperties}
   {rendererClass=&NiagaraSpriteRendererProperties, rendererIndex=1}
  ]
 }
 Grinding_Flash {
  emitterUpdateScript [
   &EmitterState(Life Cycle Mode=::NewEnumerator1, Loop Behavior=::NewEnumerator1, Loop Delay=0.28, Loop Duration Mode=::NewEnumerator1, UseLoopDelay=true)
   &SpawnRate(SpawnRate=150, Use Spawn Probability=false)
  ]
  particleSpawnScript [
   &InitializeParticle(
    Color=<rgba 50 50 50 1>
    Lifetime=&Lerp_Float(
     A=&RandomRangeFloat(Minimum=0.4, Maximum=0.6)
     B=&RandomRangeFloat(Minimum=0.4, Maximum=0.6)
     Alpha=$User.ParaGlideLeanAlpha
    )
    Mesh Scale=&SplitVector(
     Double Component Vector Channels:ECascadeNiagaraTwoVectorChannels=::NewEnumerator0
     Double Component:NiagaraFloat=&RandomRangeFloat(Minimum=0.15, Maximum=0.2)
     Single Component:NiagaraFloat=&RandomRangeFloat(Minimum=0.5, Maximum=0.8)
    )
    Mesh Scale Mode=::NewEnumerator4
    Sprite Size=&RandomRangeVector2D(Minimum=<15 5>, Maximum=<20 6.5>)
    Sprite Size Mode=::NewEnumerator4
    UsePositionOffset=false
    Write Material Random=true
   )
   &CameraOffset(Camera Offset Amount=5)
   &PartitionParticles()
  ]
  particleUpdateScript [
   &ParticleState()
   &SkeletalMeshLocation(
    Bone / Socket Index=$Particles.PartitionParticles.Partition
    Bone Sampling Mode=::NewEnumerator4
    MaskSpawnGroup=false
    Orientation Sampling=::NewEnumerator1
    Sampled Position Offset=&MergeHelper_VectorSwitch(
     0=&Lerp_Vector(
      A=&Lerp_Vector(A=<0 10 -10>, B=<0 10 10>, Alpha=$System.Audrey)
      B=&Lerp_Vector(A=<0 -10 10>, B=<0 -10 -10>, Alpha=$System.Raph)
      Alpha=&MakeFloatFromInt(Int=$Particles.PartitionParticles.Partition)
     )
     1=&Lerp_Vector(
      A=&Lerp_Vector(A=<0 9 10>, B=<0 0 15>, Alpha=$System.Audrey)
      B=&Lerp_Vector(A=<0 -9 -10>, B=<0 0 -15>, Alpha=$System.Raph)
      Alpha=&MakeFloatFromInt(Int=$Particles.PartitionParticles.Partition)
     )
     2=&Lerp_Vector(
      A=&Lerp_Vector(A=<0 30 -8>, B=<0 20 25>, Alpha=$System.Audrey)
      B=&Lerp_Vector(A=<0 -30 8>, B=<0 -20 -25>, Alpha=$System.Raph)
      Alpha=&MakeFloatFromInt(Int=$Particles.PartitionParticles.Partition)
     )
     3=&Lerp_Vector(
      A=&Lerp_Vector(A=<0 30 10>, B=<0 5 35>, Alpha=$System.Audrey)
      B=&Lerp_Vector(A=<0 -30 -10>, B=<0 -8 -30>, Alpha=$System.Raph)
      Alpha=&MakeFloatFromInt(Int=$Particles.PartitionParticles.Partition)
     )
     Index=&RandomRangeInt(Minimum=0, Maximum=3)
     Num=::NewEnumerator2
    )
    Skeletal Mesh=$System.Skeletal Mesh
   )
   &ScaleColor(
    Scale Mode=1
    Scale RGBA=&Vector4FromCurve(Vector4Curve=DataInterface() {propertyValues {
       AlphaCurve {keys [<at 0 1>,<at 1 0>]}
       BlueCurve {keys [<at 0 0>,<at 1 0>]}
       GreenCurve {keys [<at 0 0.1716>,<at 1 0.0396>]}
       RedCurve {keys [<at 0 1>,<at 1 1>]}
      }})
    ScaleRGBA=true
   ) {module=ScaleColor001}
   &ScaleSpriteSize(Uniform Curve Sprite Scale=DataInterface() {propertyValues {Curve {defaultValue=0, keys [{tangentMode=RCTM_User, time=0, value=1}, {tangentMode=RCTM_User, time=1, value=0}]}}})
   &SolveForcesAndVelocity(Clamp Velocity=false, Limit Acceleration=false)
  ]
  rendererClasses [&NiagaraSpriteRendererProperties]
  renderers [{rendererClass=&NiagaraSpriteRendererProperties}]
 }
 Nuts_and_Bolts {
  emitterUpdateScript [
   &EmitterState()
   &SpawnRate(SpawnRate=&Lerp_Float(A=30, B=30, Alpha=$System.ParaGlideAlphaFloat), Use Spawn Probability=false)
  ]
  particleSpawnScript [
   &PartitionParticles(Number Of Partitions=4)
   &InitializeParticle(Lifetime=2, Sprite Size Mode=::NewEnumerator3, Uniform Sprite Size=15, UsePositionOffset=false, Write Material Random=true)
   &SkeletalMeshLocation(
    Bone / Socket Index=$Particles.PartitionParticles.Partition
    Bone Sampling Mode=::NewEnumerator4
    MaskSpawnGroup=false
    Orientation Sampling=::NewEnumerator1
    Sampled Position Offset=&SelectVectorFromArray(
     Direct Array Index=$Particles.PartitionParticles.Partition
     Vector Selection Array=DataInterface() {propertyValues {FloatData [<5 0 0>, <-5 0 0>, <-8 0 2>, <8 0 -2>]}}
    )
    Skeletal Mesh=$System.Skeletal Mesh
   ) {module=SkeletalMeshLocation001}
   &AddVelocity(
    Cone Angle=20
    Cone Axis=<0 0 1>
    Module.UseConeFalloff=false
    Module.UseVelDistribution=false
    Override Seed=false
    Velocity Mode=::NewEnumerator2
    Velocity Speed=&Lerp_Float(A=265, B=500, Alpha=$System.ParaGlideAlphaFloat)
   )
   /SetVariables(Particles.VisibilityTag=&RandomRangeInt(Minimum=0, Maximum=1)) {isSetParameters=true}
   &CameraOffset()
  ]
  particleUpdateScript [
   &ParticleState()
   &UpdateMeshOrientation(Rotation Rate=&RandomRangeFloat(Minimum=1, Maximum=1))
   &SolveForcesAndVelocity(Clamp Velocity=false, Limit Acceleration=false)
   &Color(Color=&LerpLinearColors(
     StartColor=&ColorFromCurve(ColorCurve=DataInterface() {propertyValues {
        AlphaCurve {keys [<at 0 1 RCIM_Cubic>,<at 1 0 RCIM_Cubic>]}
        BlueCurve {keys [
          <at 0 0.0459>
          {arriveTangent=1.0773, interpMode=RCIM_Cubic, leaveTangent=1.0773, time=0.0216, value=0.1081}
          <at 0.184 0.2441>
          <at 0.3015 0.3802>
         ]}
        GreenCurve {keys [
          <at 0 0.571>
          {arriveTangent=-1.4646, interpMode=RCIM_Cubic, leaveTangent=-1.4646, time=0.0216, value=0.3701}
          <at 0.184 0.3015>
          <at 0.3015 0.3802>
         ]}
        RedCurve {keys [<at 0 1>,<at 0.0216 1 RCIM_Cubic>,<at 0.184 1>,<at 0.3015 0.3802>]}
       }})
     EndColor=&ColorFromCurve(ColorCurve=DataInterface() {propertyValues {
        AlphaCurve {keys [<at 0 1 RCIM_Cubic>,<at 1 0 RCIM_Cubic>]}
        BlueCurve {keys [{arriveTangent=3.1647, interpMode=RCIM_Cubic, time=0, value=0.1081}, <at 0.0626 0.2441>, <at 0.1189 0.3802>]}
        GreenCurve {keys [{arriveTangent=-4.3025, interpMode=RCIM_Cubic, time=0, value=0.3701}, <at 0.0626 0.3015>, <at 0.1189 0.3802>]}
        RedCurve {keys [<at 0 1 RCIM_Cubic>,<at 0.0626 1>,<at 0.1189 0.3802>]}
       }})
     LerpFactor=&MergeHelper_FloatSwitch(0=0, 1=0, 2=1, 3=1, Index=$Particles.PartitionParticles.Partition, Num=::NewEnumerator2)
    ))
   /SetVariables(Particles.GlowColor=&MultiplyLinearColor(Linear Color=$Particles.Color, Scale Factor=&MakeVector4_FromLinearColor(LinearColor=<rgba 1 1 1 0.02>))) {isSetParameters=true}
  ]
  rendererClasses [&NiagaraMeshRendererProperties]
  renderers [
   {rendererClass=&NiagaraMeshRendererProperties, rendererIndex=0}
   {rendererClass=&NiagaraMeshRendererProperties, rendererIndex=1}
  ]
 }
 ParticleDecal {
  emitterUpdateScript [
   &EmitterState(
    Inactive Response=::NewEnumerator1
    Life Cycle Mode=::NewEnumerator1
    Loop Behavior=::NewEnumerator1
    Loop Delay=1.4
    Loop Duration Mode=::NewEnumerator1
    UseLoopDelay=false
   )
   &SpawnBurst_Instantaneous(Spawn Count=4, Use Spawn Probability=false, UseLoopCountLimit=false)
  ]
  particleSpawnScript [
   &FN_SetSpawnGroup()
   &InitializeParticle(
    Color=<rgba 0.2344 0.0714 0 2>
    Sprite Rotation Mode=::NewEnumerator2
    Sprite Size Mode=::NewEnumerator3
    Uniform Sprite Size=&SelectFloatFromArray(Direct Array Index=$Particles.SpawnGroup, Float Selection Array=DataInterface() {propertyValues {FloatData [70,70,100,100]}})
    UsePositionOffset=false
    Write Material Random=true
   ) {module=InitializeParticle001}
  ]
  particleUpdateScript [
   &SkeletalMeshLocation(
    Bone / Socket Index=$Particles.SpawnGroup
    Bone Sampling Mode=::NewEnumerator4
    MaskSpawnGroup=false
    Orientation Sampling=::NewEnumerator1
    Sampled Position Offset=&MergeHelper_VectorSwitch(
     0=&Lerp_Vector(A=<0 18 2>, B=<0 5 18>, Alpha=$System.Audrey)
     1=&Lerp_Vector(A=<0 -18 -2>, B=<0 -8 -20>, Alpha=$System.Raph)
     2=<-20 10 0>
     3=<20 -10 0>
     Index=&SpawnGroup()
     Num=::NewEnumerator2
    )
    Skeletal Mesh=$System.Skeletal Mesh
   ) {module=SkeletalMeshLocation001}
   &ParticleState(Kill Particles When Lifetime Has Elapsed=false, Let Infinitely Lived Particles Die When Emitter Deactivates=true)
   &ScaleSpriteSize(Scale Factor=&Vector2DFromFloat(Value=&FloatFromCurve(FloatCurve=DataInterface() {propertyValues {Curve {keys [
          {interpMode=RCIM_Cubic, leaveTangent=5.3269, tangentMode=RCTM_Break, time=0, value=0}
          {arriveTangent=0.6, interpMode=RCIM_Cubic, leaveTangent=0.6, time=0.2, value=0.54}
          {arriveTangent=0.0611, interpMode=RCIM_Cubic, tangentMode=RCTM_Break, time=1, value=0.6}
         ]}}})))
   &FN_Light_Attributes(
    Color=<rgba 40 19.7333 2 1>
    Create Color Attribute=true
    Create Exponent Attribute=true
    Create Radius Attribute=true
    Exponent=2
    Radius=70
    Scale Color=&FN_FrontEndFloatModifier(Gameplay=1, Front End=0.2)
   )
   &ScaleColor(Scale Alpha=$System.RevealCurve, Scale RGB=&VectorFromFloat(Value=30), ScaleA=true, ScaleRGB=true)
   &CameraOffset(Camera Offset Amount=20)
  ]
  rendererClasses [&NiagaraSpriteRendererProperties]
  renderers [{rendererClass=&NiagaraSpriteRendererProperties}]
 }
 SparkPosition {
  emitterUpdateScript [
   &EmitterState(Life Cycle Mode=::NewEnumerator1, Loop Behavior=::NewEnumerator1, Loop Duration Mode=::NewEnumerator1, UseLoopDelay=false)
   &SpawnBurst_Instantaneous(Spawn Count=2, Use Spawn Probability=false, UseLoopCountLimit=false)
   &SpawnBurst_Instantaneous(Spawn Count=2, Spawn Group=1, Use Spawn Probability=false, UseLoopCountLimit=false) {module=SpawnBurst_Instantaneous001}
   &SpawnBurst_Instantaneous(Spawn Count=2, Spawn Group=2, Use Spawn Probability=false, UseLoopCountLimit=false) {module=SpawnBurst_Instantaneous002}
   &SpawnBurst_Instantaneous(Spawn Count=2, Spawn Group=3, Use Spawn Probability=false, UseLoopCountLimit=false) {module=SpawnBurst_Instantaneous003}
  ]
  particleSpawnScript [
   &InitializeParticle(
    Color Mode=::NewEnumerator0
    Lifetime=5
    Mesh Scale Mode=::NewEnumerator3
    Sprite Rotation Mode=::NewEnumerator0
    Sprite Size Mode=::NewEnumerator3
    Uniform Sprite Size=15
    UsePositionOffset=false
    Write Material Random=true
   )
   &PartitionParticles()
   &InitialMeshOrientation(
    Mesh Orientation Mode=::NewEnumerator2
    Orientation Vector=<0 0 1>
    Rotation=&SelectVectorFromArray(
     Direct Array Index=$Particles.PartitionParticles.Partition
     Vector Selection Array=DataInterface() {propertyValues {FloatData [<0 0 -0.25>, <0 0 0.25>]}}
    )
    Rotation Coordinate Space=::NewEnumerator2
    Use Rotation Vector=true
   )
   &ApplyInitialForces()
   /SetVariables() {isSetParameters=true, module=SetVariables_#9}
   /SetVariables(Particles.VisibilityTag=&SpawnGroup()) {isSetParameters=true, module=SetVariables_#10}
   &DynamicMaterialParameters(Module.Index0Float0Write=true, Module.Index0Float1Write=true, Module.Index0Float2Write=true, Module.Index0Float3Write=true)
   /SetVariables() {isSetParameters=true}
  ]
  particleUpdateScript [
   /SetVariables(Particles.TimeTracker:NiagaraFloat=&Multiply_Float(A=&Add_Float(A=$Engine.DeltaTime, B=$Particles.TimeTracker), B=1)) {isSetParameters=true, module=SetVariables_#11}
   &ParticleState(Kill Particles When Lifetime Has Elapsed=false)
   &ScaleMeshSize(Scale Factor=&VectorFromFloat(Value=1))
   &SkeletalMeshLocation(
    Bone / Socket Index=$Particles.PartitionParticles.Partition
    Bone Sampling Mode=::NewEnumerator4
    MaskSpawnGroup=false
    Orientation Sampling=::NewEnumerator1
    Sampled Position Offset=&MergeHelper_VectorSwitch(
     0=&Lerp_Vector(
      A=&Lerp_Vector(A=<0 10 -10>, B=<0 10 10>, Alpha=$System.Audrey)
      B=&Lerp_Vector(A=<0 -10 10>, B=<0 -10 -10>, Alpha=$System.Raph)
      Alpha=&MakeFloatFromInt(Int=$Particles.PartitionParticles.Partition)
     )
     1=&Lerp_Vector(
      A=&Lerp_Vector(A=<0 9 10>, B=<0 0 15>, Alpha=$System.Audrey)
      B=&Lerp_Vector(A=<0 -9 -10>, B=<0 0 -15>, Alpha=$System.Raph)
      Alpha=&MakeFloatFromInt(Int=$Particles.PartitionParticles.Partition)
     )
     2=&Lerp_Vector(
      A=&Lerp_Vector(A=<0 30 -8>, B=<0 20 25>, Alpha=$System.Audrey)
      B=&Lerp_Vector(A=<0 -30 8>, B=<0 -20 -25>, Alpha=$System.Raph)
      Alpha=&MakeFloatFromInt(Int=$Particles.PartitionParticles.Partition)
     )
     3=&Lerp_Vector(
      A=&Lerp_Vector(A=<0 30 10>, B=<0 5 35>, Alpha=$System.Audrey)
      B=&Lerp_Vector(A=<0 -30 -10>, B=<0 -8 -30>, Alpha=$System.Raph)
      Alpha=&MakeFloatFromInt(Int=$Particles.PartitionParticles.Partition)
     )
     Index=&SpawnGroup()
     Num=::NewEnumerator2
    )
    Skeletal Mesh=$System.Skeletal Mesh
   )
   &SpriteFacingAndAlignment(
    Sprite Facing=&RotateVectorByQuaternion(
     Vector To Rotate=&MergeHelper_VectorSwitch(
      0=&Lerp_Vector(
       A=&Lerp_Vector(A=<0 -1 -1>, B=<0 1 -1>, Alpha=$System.Audrey)
       B=&Lerp_Vector(A=<0 1 1>, B=<0 -0.5 1>, Alpha=$System.Raph)
       Alpha=&MakeFloatFromInt(Int=$Particles.PartitionParticles.Partition)
      )
      1=&Lerp_Vector(
       A=&Lerp_Vector(A=<0 -1 1>, B=<0 -1 0>, Alpha=$System.Audrey)
       B=&Lerp_Vector(A=<0 1 -1>, B=<0 1 0>, Alpha=$System.Raph)
       Alpha=&MakeFloatFromInt(Int=$Particles.PartitionParticles.Partition)
      )
      2=&Lerp_Vector(
       A=&Lerp_Vector(A=<0 1 -1>, B=<0 1 0.2>, Alpha=$System.Audrey)
       B=&Lerp_Vector(A=<0 -1 1>, B=<0 -1 0>, Alpha=$System.Raph)
       Alpha=&MakeFloatFromInt(Int=$Particles.PartitionParticles.Partition)
      )
      3=&Lerp_Vector(
       A=&Lerp_Vector(A=<0 1 1>, B=<0 -0.5 1>, Alpha=$System.Audrey)
       B=&Lerp_Vector(A=<0 -1 -1>, B=<0 0.5 -1>, Alpha=$System.Raph)
       Alpha=&MakeFloatFromInt(Int=$Particles.PartitionParticles.Partition)
      )
      Index=&SpawnGroup()
      Num=::NewEnumerator2
     )
     Quaternion=$Particles.SkeletalMeshLocation.SampledRotation
    )
    WriteAlignment=true
    WriteFacing=true
   )
   &SolveForcesAndVelocity(Clamp Velocity=false, Limit Acceleration=false)
   /SetVariables(Particles.Velocity=$Particles.SpriteFacing) {isSetParameters=true}
  ]
  rendererClasses [&NiagaraSpriteRendererProperties]
  renderers [
   {rendererClass=&NiagaraSpriteRendererProperties}
   {rendererClass=&NiagaraSpriteRendererProperties, rendererIndex=1}
   {rendererClass=&NiagaraSpriteRendererProperties, rendererIndex=2}
   {rendererClass=&NiagaraSpriteRendererProperties, rendererIndex=3}
  ]
 }
 Sparks {
  emitterUpdateScript [
   &EmitterState(Life Cycle Mode=::NewEnumerator1, Loop Behavior=::NewEnumerator1, Loop Delay=0.5, Loop Duration Mode=::NewEnumerator1, UseLoopDelay=true)
   &SpawnRate(SpawnRate=&Lerp_Float(A=20, B=100, Alpha=$User.ParaGlideLeanAlpha), Use Spawn Probability=false)
  ]
  particleSpawnScript [
   &PartitionParticles(Number Of Partitions=4)
   &InitializeParticle(
    Lifetime=&Lerp_Float(
     A=&RandomRangeFloat(Minimum=0.6, Maximum=0.8)
     B=&RandomRangeFloat(Minimum=0.3, Maximum=0.5)
     Alpha=$User.ParaGlideLeanAlpha
    )
    Mesh Scale=&SplitVector(
     Double Component Vector Channels:ECascadeNiagaraTwoVectorChannels=::NewEnumerator0
     Double Component:NiagaraFloat=&RandomRangeFloat(Minimum=0.15, Maximum=0.2)
     Single Component:NiagaraFloat=&RandomRangeFloat(Minimum=0.5, Maximum=0.8)
    )
    Mesh Scale Mode=::NewEnumerator4
    Sprite Size=&RandomRangeVector2D(Minimum=<0.5 2>, Maximum=<1 4>)
    Sprite Size Mode=::NewEnumerator4
    UsePositionOffset=false
    Write Material Random=true
   )
   &SkeletalMeshLocation(
    Bone / Socket Index=$Particles.PartitionParticles.Partition
    Bone Sampling Mode=::NewEnumerator4
    MaskSpawnGroup=false
    Position Sampling=::NewEnumerator1
    Skeletal Mesh=$System.Skeletal Mesh
   ) {module=SkeletalMeshLocation001}
   &ShapeLocation(
    MaskSpawnGroup=false
    Override Seed=false
    Shape Origin=&AddVectorToPosition(Position=$Output.SkeletalMeshLocation001.SampledPositionWithoutOffset, Vector=<0 0 -10>)
    Sphere Radius=20
    UseOwnerScale=true
   )
   &DynamicMaterialParameters(
    Index 0 Param 1=&RandomRangeFloat(Minimum=0, Maximum=0.9)
    Index 0 Param 2=9
    Index 0 Param 4=&RandomRangeFloat(Minimum=0, Maximum=1)
    Module.Param0WriteEnabled=true
    Module.Param1WriteEnabled=true
    Module.Param2WriteEnabled=true
    Module.Param3WriteEnabled=true
   )
   &AddVelocity(
    Rotation Coordinate Space=::World
    Velocity=<0 0 100>
    Velocity Speed Scale=&Lerp_Float(
     A=&RandomRangeFloat(Minimum=1, Maximum=2)
     B=3.5
     Alpha=&ClampFloat(Float=$User.ParaGlideLeanAlpha)
    )
   ) {module=AddVelocity001}
   &AddVelocity(
    Velocity=&MergeHelper_VectorSwitch(0=<100 0 0>, 1=<-100 0 0>, 2=<20 0 0>, 3=<-20 0 0>, Index=$Particles.PartitionParticles.Partition, Num=::NewEnumerator2)
    Velocity Speed Scale=&FN_FrontEndFloatModifier(Gameplay=1, Front End=0)
   )
   &AddVelocity(
    Velocity=<0 -50 0>
    Velocity Speed Scale=&FN_FrontEndFloatModifier(Gameplay=0, Front End=$User.ParaGlideLeanAlpha)
   ) {module=AddVelocity002}
   /SetVariables(Particles.SpawnRotationVector=&RandomRangeVector(Minimum=<0 0 0>, Maximum=<1 1 1>)) {isSetParameters=true}
  ]
  particleUpdateScript [
   &ParticleState()
   &CurlNoiseForce(
    Noise Frequency=100
    Noise Strength=&Multiply_Float(A=2000, B=&Lerp_Float(A=0.3, B=1, Alpha=&ClampFloat(Float=$User.ParaGlideLeanAlpha)))
    Pan Noise=true
   )
   &SolveForcesAndVelocity(Clamp Velocity=false, Limit Acceleration=false)
   /SetVariables(Particles.Glow03_Color=&MakeLinearColorFromVectorAndFloat(
     "Vector (RGB)"=&MakeVectorFromLinearColorRGB(Color=&MultiplyLinearColor(Linear Color=<rgba 1.5 0.05 0 1>, Scale Factor=&MakeVector4_FromFloat(Float=1)))
     "Float (Alpha)"=&Multiply_Float(
      A=0.04
      B=&FloatFromCurve(FloatCurve=DataInterface() {propertyValues {Curve {keys [<at 0 0>,<at 0.2 1>]}}})
     )
    )) {isSetParameters=true, module=SetVariables_#12}
   &ScaleColor(
    Scale Alpha=&FloatFromCurve(FloatCurve=DataInterface() {propertyValues {Curve {keys [<at 0 0>,<at 0.2 1>,<at 0.4 1>,<at 1 1>]}}})
    Scale RGB=&MakeVectorFromLinearColorRGB(Color=&ColorFromCurve(ColorCurve=DataInterface() {propertyValues {
        AlphaCurve {keys [<at 0 1 RCIM_Cubic>,<at 1 0 RCIM_Cubic>]}
        BlueCurve {keys [<at 0.1638 0 RCIM_Cubic>,<at 0.498 0 RCIM_Cubic>]}
        GreenCurve {keys [<at 0.1638 1.4 RCIM_Cubic>,<at 0.498 0.05 RCIM_Cubic>]}
        RedCurve {keys [<at 0.1638 3 RCIM_Cubic>,<at 0.498 1 RCIM_Cubic>]}
       }}))
    ScaleA=true
    ScaleRGB=true
   )
   /SetVariables(Particles.Glow03_Size=&Multiply_Vector2D_ByFloat(
     Float=&FloatFromCurve(
      FloatCurve=DataInterface() {propertyValues {Curve {keys [
          <at 0 0 RCIM_Cubic>
          <at 0.2 1.3 RCIM_Cubic>
          {arriveTangent=-0.1423, interpMode=RCIM_Cubic, leaveTangent=-0.1423, tangentMode=RCTM_User, time=1, value=0}
         ]}}}
      Scale Curve=10
     )
     Vector2D=<0.6 1.2>
    )) {isSetParameters=true, module=SetVariables_#13}
  ]
  rendererClasses [&NiagaraSpriteRendererProperties]
  renderers [
   {rendererClass=&NiagaraSpriteRendererProperties}
   {rendererClass=&NiagaraSpriteRendererProperties, rendererIndex=1}
  ]
 }
 Trails {
  emitterUpdateScript [
   &EmitterState(
    Inactive Response=::NewEnumerator1
    Life Cycle Mode=::NewEnumerator1
    Loop Behavior=::NewEnumerator1
    Loop Duration Mode=::NewEnumerator1
    UseLoopDelay=false
   )
   &SpawnBurst_Instantaneous(Spawn Count=4, Use Spawn Probability=false, UseLoopCountLimit=false)
   /SetVariables(
    Emitter.CameraFadeDistance:NiagaraFloat=&Lerp_Float(A=200, B=200, Alpha=$User.ParaGlideLeanAlpha)
    Emitter.EmitterDepthFade:NiagaraFloat=&Lerp_Float(A=0, B=150, Alpha=$System.ParaGlideAlphaFloat)
   ) {isSetParameters=true}
  ]
  particleSpawnScript [
   &PartitionParticles(Number Of Partitions=4)
   &InitializeParticle(
    Mesh Scale=&SelectVectorFromArray(
     Direct Array Index=$Particles.PartitionParticles.Partition
     Vector Selection Array=DataInterface() {propertyValues {FloatData [<0.5 0.5 1.6>, <0.5 0.5 1.6>, <0.2 0.2 0.5>, <0.2 0.2 0.5>]}}
    )
    Mesh Scale Mode=::NewEnumerator4
    Sprite Size Mode=::NewEnumerator3
    Uniform Sprite Size=20
    UsePositionOffset=false
    Write Material Random=true
   )
   &InitialMeshOrientation(
    Mesh Orientation Mode=::NewEnumerator2
    Rotation=&SelectVectorFromArray(
     Direct Array Index=$Particles.PartitionParticles.Partition
     Vector Selection Array=DataInterface() {propertyValues {FloatData [<0 0 0>, <0 0 0>, <0 0.03 0>, <0 0.03 0>]}}
    )
    Use Rotation Vector=true
   )
   &InitialMeshOrientation(
    Mesh Orientation Mode=::NewEnumerator2
    Rotation=&MergeHelper_VectorSwitch(0=<-0.05 0.05 0>, 1=<-0.12 0.1 0.5>, 2=<0 0 0>, 3=<0 0 0.5>, Index=$Particles.PartitionParticles.Partition, Num=::NewEnumerator2)
    Use Rotation Vector=true
   ) {module=InitialMeshOrientation002}
   /SetVariables(Particles.VisibilityTag=&SelectIntFromArray(Direct Array Index=$Particles.PartitionParticles.Partition, Int Selection Array=DataInterface() {propertyValues {IntData [0,0,1,1]}})) {isSetParameters=true}
   &CameraOffset(Camera Offset Amount=15)
  ]
  particleUpdateScript [
   &SkeletalMeshLocation(
    Bone / Socket Index=$Particles.PartitionParticles.Partition
    Bone Sampling Mode=::NewEnumerator4
    MaskSpawnGroup=false
    Orientation Sampling=::NewEnumerator1
    Sampled Position Offset=&MergeHelper_VectorSwitch(
     0=&Lerp_Vector(A=<5 0 0>, B=<-10 10 0>, Alpha=$System.ParaGlideAlphaFloat)
     1=&Lerp_Vector(A=<-5 0 0>, B=<10 -10 0>, Alpha=$System.ParaGlideAlphaFloat)
     2=<1 0 0>
     3=<-1 0 0>
     Index=$Particles.PartitionParticles.Partition
     Num=::NewEnumerator2
    )
    Skeletal Mesh=$System.Skeletal Mesh
   ) {module=SkeletalMeshLocation001}
   &ParticleState(Kill Particles When Lifetime Has Elapsed=false, Let Infinitely Lived Particles Die When Emitter Deactivates=true)
   &Timeline(
    Play Rate=&Lerp_Float(
     A=&MergeHelper_FloatSwitch(0=2.5, 1=2.5, 2=2.7, 3=2.7, Index=$Particles.PartitionParticles.Partition, Num=::NewEnumerator2)
     B=4
     Alpha=&ClampFloat(Float=$User.ParaGlideLeanAlpha)
    )
    Timeline End=10000000
   )
   &ScaleMeshSize(Scale Factor=&MergeHelper_VectorSwitch(
     0=&Multiply_Vector(
      A=&Multiply_Vector(A=<1 1 1>, B=&Lerp_Vector(A=<0.8 0.8 0.8>, B=<1.6 1.6 1>, Alpha=$User.ParaGlideLeanAlpha))
      B=&MakeVector(
       X=&Lerp_Float(A=1, B=1.3, Alpha=&ClampFloat(Float=$User.ParaGlideLeanAlpha))
       Y=&Lerp_Float(A=1, B=1.3, Alpha=&ClampFloat(Float=$User.ParaGlideLeanAlpha))
       Z=&Lerp_Float(A=0.8, B=1.5, Alpha=&ClampFloat(Float=$User.ParaGlideLeanAlpha))
      )
     )
     1=&Multiply_Vector(
      A=&Multiply_Vector(A=<1 1 1>, B=&Lerp_Vector(A=<0.8 0.8 0.8>, B=<1.6 1.6 1>, Alpha=$User.ParaGlideLeanAlpha))
      B=&MakeVector(
       X=&Lerp_Float(A=1, B=1, Alpha=&ClampFloat(Float=$User.ParaGlideLeanAlpha))
       Y=&Lerp_Float(A=1, B=1.3, Alpha=&ClampFloat(Float=$User.ParaGlideLeanAlpha))
       Z=&Lerp_Float(A=0.8, B=1.5, Alpha=&ClampFloat(Float=$User.ParaGlideLeanAlpha))
      )
     )
     2=&Multiply_Vector(
      A=&Lerp_Vector(A=<1 1 1>, B=<2 2 1.3>, Alpha=$User.ParaGlideLeanAlpha)
      B=&MakeVector(X=1, Y=1, Z=&Lerp_Float(A=1, B=2, Alpha=&ClampFloat(Float=$User.ParaGlideLeanAlpha)))
     )
     3=&Multiply_Vector(
      A=&Lerp_Vector(A=<1 1 1>, B=<2 2 1.3>, Alpha=$User.ParaGlideLeanAlpha)
      B=&MakeVector(X=1, Y=1, Z=&Lerp_Float(A=1, B=2, Alpha=&ClampFloat(Float=$User.ParaGlideLeanAlpha)))
     )
     Index=$Particles.PartitionParticles.Partition
     Num=::NewEnumerator2
    ))
   &ScaleMeshSize(Scale Factor=&SplitVector(
     Double Component Vector Channels:ECascadeNiagaraTwoVectorChannels=::NewEnumerator0
     Double Component:NiagaraFloat=&Waveform("Waveform [1]"=::NewEnumerator0, "[1] Frequency"=1.35, "[1] Amplitude Min/Max"=<0.7 1.3>)
     Single Component:NiagaraFloat=1
    )) {module=ScaleMeshSize001}
   &ScaleColor(
    Scale Alpha=&Multiply_Float(
     A=&Lerp_Float(A=1, B=&FN_FrontEndFloatModifier(Gameplay=0.7, Front End=1), Alpha=$User.ParaGlideLeanAlpha)
     B=$System.ToD
    )
    Scale RGB=&Multiply_VectorByFloat(Vector=<1 1 1>, Float=&FN_FrontEndFloatModifier(Gameplay=1, Front End=1))
    ScaleA=true
    ScaleRGB=true
   )
   /SetVariables(Particles.FireMeshOrientation:Quat4f=$Particles.MeshOrientation) {isSetParameters=true}
   &InterpolateOverTime(
    Rate Of Change=10
    Target Vector Value:Vector3f=&TransformPosition(Position=$Particles.Position, Source Space=::Local, Destination Space=::World)
    Variable Type=::NewEnumerator2
   )
   &DynamicMaterialParameters(
    Index 0 Param 1=$Particles.Timeline.Time
    Index 0 Param 2=&Lerp_Float(
     A=&MergeHelper_FloatSwitch(
      0=&FN_FrontEndFloatModifier(Gameplay=7, Front End=5)
      1=&FN_FrontEndFloatModifier(Gameplay=7, Front End=5)
      2=&FN_FrontEndFloatModifier(Gameplay=2, Front End=5)
      3=&FN_FrontEndFloatModifier(Gameplay=2, Front End=5)
      Index=$Particles.PartitionParticles.Partition
      Num=::NewEnumerator2
     )
     B=0
     Alpha=&ClampFloat(Float=$User.ParaGlideLeanAlpha)
    )
    Index 0 Param 3=&Multiply_Float(
     A=&MakeCustomFloatFromBool(
      Bool=$User.bIsFrontEndPreview
      True Float=&Lerp_Float(A=0.9, B=1.3, Alpha=$User.ParaGlideLeanAlpha)
      False Float=&Lerp_Float(A=1, B=1.1, Alpha=$User.ParaGlideLeanAlpha)
     )
     B=&MergeHelper_FloatSwitch(0=1, 1=1, 2=0.7, 3=0.7, Index=$Particles.PartitionParticles.Partition, Num=::NewEnumerator2)
    )
    Index 0 Param 4=&Lerp_Float(A=0, B=0, Alpha=$User.ParaGlideLeanAlpha)
    Index 1 Param 1=$System.RevealCurve
    Index 3 Param 4=&Lerp_Float(A=1, B=0, Alpha=$User.ParaGlideLeanAlpha)
    Index 3 Vector 1:Vector3f=$StackContext.InterpolateOverTime.InterpolateOverTime_V3_.Moving Average
    Module.Index0Float0Write=true
    Module.Index0Float1Write=true
    Module.Index0Float2Write=true
    Module.Index0Float3Write=false
    Module.Index1Float0Write=true
    Module.Index1Float1Write=false
    Module.Index1Float2Write=false
    Module.Index1Float3Write=false
    Module.Index3Float3Write=true
    Module.Index3VectorWrite=true
    Write Parameter Index 1=::NewEnumerator1
    Write Parameter Index 3=::NewEnumerator3
   ) {module=DynamicMaterialParameters001}
   /SetVariables(Particles.DMP1:Vector4f=&MakeVector4(
     X=0
     Y=&Lerp_Float(
      A=&FN_FrontEndFloatModifier(Gameplay=7, Front End=5)
      B=0
      Alpha=&ClampFloat(Float=$User.ParaGlideLeanAlpha)
     )
     Z=&MakeCustomFloatFromBool(
      Bool=$User.bIsFrontEndPreview
      True Float=&Lerp_Float(A=0.9, B=1.3, Alpha=$User.ParaGlideLeanAlpha)
      False Float=&Lerp_Float(A=1, B=1.3, Alpha=$User.ParaGlideLeanAlpha)
     )
     W=&Lerp_Float(A=1, B=2, Alpha=$User.ParaGlideLeanAlpha)
    )) {isSetParameters=true}
   /SetVariables(
    Particles.BGColor:LinearColor=<rgba 0.01 0.01 0.01 0.5>
    Particles.BGOrienation:Quat4f=&MultiplyQuaternion(
     Quaternion A=$StackContext.MeshOrientation
     Quaternion B=&MakeQuaternion(Quaternion From=0, Axis=<0 0 1>, Angle=180, From=?, To=?, Angle Type=::NewEnumerator0, Normalize Vectors=false)
    )
   ) {isSetParameters=true}
   /SetVariables(Particles.GlowColor=<rgba 1 0.7364 0 0.2>) {isSetParameters=true}
   /SetVariables(
    Particles.BottomFireScale:Vector3f=&Multiply_Vector(
     A=&Multiply_VectorByFloat(
      Vector=<3.1 3.1 10>
      Float=&MergeHelper_FloatSwitch(0=0.22, 1=0.22, 2=0.18, 3=0.18, Index=$Particles.PartitionParticles.Partition, Num=::NewEnumerator2)
     )
     B=&Lerp_Vector(
      A=<1 1 1>
      B=&FN_FrontEndVector3Modifier(Gameplay=<1.3 1.3 1.5>, Front End=<1.1 1.1 1>)
      Alpha=$User.ParaGlideLeanAlpha
     )
    )
    Particles.FireColor:LinearColor=&MakeLinearColorFromVectorAndFloat(
     "Vector (RGB)"=<1 1 1>
     "Float (Alpha)"=&Multiply_Float(
      A=&Multiply_Float(
       A=$System.FireRevealCurve
       B=&Lerp_Float(A=1, B=&FN_FrontEndFloatModifier(Gameplay=1, Front End=1), Alpha=$User.ParaGlideLeanAlpha)
      )
      B=$System.ToD
     )
    )
    Particles.InnerFireScale:Vector3f=&Multiply_Vector(
     A=&Multiply_VectorByFloat(Vector=<1.5 1 1.5>, Float=0.22)
     B=&Lerp_Vector(
      A=<1.5 1.5 1>
      B=&FN_FrontEndVector3Modifier(Gameplay=<1 1 1.5>, Front End=<1.1 1 1.1>)
      Alpha=$User.ParaGlideLeanAlpha
     )
    )
   ) {isSetParameters=true}
   /SetVariables(Particles.MeshOrientation=&MultiplyQuaternion(
     Quaternion A=$Particles.Initial.MeshOrientation
     Quaternion B=&MultiplyQuaternion(
      Quaternion A=&MakeQuaternion(
       Quaternion From=0
       Axis=<1 0 0>
       Angle=&FN_FrontEndFloatModifier(
        Gameplay=&MergeHelper_FloatSwitch(
         0=&Lerp_Float(A=20, B=10, Alpha=$System.CameraCalculationFloat)
         1=&Lerp_Float(A=-22, B=-10, Alpha=$System.CameraCalculationFloat)
         2=0
         3=0
         Index=$Particles.PartitionParticles.Partition
         Num=::NewEnumerator2
        )
        Front End=&Lerp_Float(
         A=&MergeHelper_FloatSwitch(0=20, 1=-22, 2=0, 3=0, Index=$Particles.PartitionParticles.Partition, Num=::NewEnumerator2)
         B=&MergeHelper_FloatSwitch(0=30, 1=-30, 2=20, 3=-20, Index=$Particles.PartitionParticles.Partition, Num=::NewEnumerator2)
         Alpha=$User.ParaGlideLeanAlpha
        )
       )
       From=?
       To=?
       Angle Type=::NewEnumerator0
       Normalize Vectors=false
      )
      Quaternion B=&LerpQuaternion(
       Quaternion A:Quat4f=<0 0 0 1>
       Quaternion B:Quat4f=&MultiplyQuaternion(
        Quaternion A=&MakeQuaternion(
         Quaternion From=0
         Axis=<1 0 0>
         Angle=&MergeHelper_FloatSwitch(
          0=&Lerp_Float(A=0, B=10, Alpha=$System.MaskedAudrey)
          1=&Lerp_Float(A=0, B=-35, Alpha=$System.MaskedRaph)
          2=0
          3=0
          Index=$Particles.PartitionParticles.Partition
          Num=::NewEnumerator2
         )
         From=?
         To=?
         Angle Type=::NewEnumerator0
         Normalize Vectors=false
        )
        Quaternion B=&MakeQuaternion(
         Quaternion From=0
         Axis=<0 1 0>
         Angle=&MergeHelper_FloatSwitch(
          0=&Lerp_Float(A=0, B=-20, Alpha=$System.MaskedAudrey)
          1=&Lerp_Float(A=0, B=0, Alpha=$System.MaskedRaph)
          2=0
          3=0
          Index=$Particles.PartitionParticles.Partition
          Num=::NewEnumerator2
         )
         From=?
         To=?
         Angle Type=::NewEnumerator0
         Normalize Vectors=false
        )
       )
       Lerp Factor:NiagaraFloat=0
       Enable Spherical Linear Interpolation:NiagaraBool=false
      )
     )
    )) {isSetParameters=true}
  ]
  rendererClasses [&NiagaraMeshRendererProperties]
  renderers [
   {rendererClass=&NiagaraMeshRendererProperties, rendererIndex=0}
   {rendererClass=&NiagaraMeshRendererProperties, rendererIndex=1}
   {rendererClass=&NiagaraMeshRendererProperties, rendererIndex=2}
   {rendererClass=&NiagaraMeshRendererProperties, rendererIndex=3}
  ]
 }
}
summary {
 systemName=NS_CozyMechanic_Contrail_Save
 emitters [
  {emitterName=ParticleDecal, rendererClasses [&NiagaraSpriteRendererProperties]}
  {emitterName=GlowAndLight, rendererClasses [&NiagaraLightRendererProperties, &NiagaraSpriteRendererProperties]}
  {emitterName=Sparks, rendererClasses [&NiagaraSpriteRendererProperties]}
  {emitterName=Trails, rendererClasses [&NiagaraMeshRendererProperties]}
  {emitterName=Gear, rendererClasses [&NiagaraMeshRendererProperties]}
  {emitterName=Griding_Sparks, rendererClasses [&NiagaraSpriteRendererProperties]}
  {emitterName=SparkPosition, rendererClasses [&NiagaraSpriteRendererProperties]}
  {emitterName=Nuts_and_Bolts, rendererClasses [&NiagaraMeshRendererProperties]}
  {emitterName=Grinding_Flash, rendererClasses [&NiagaraSpriteRendererProperties]}
  {bEnabled=false, emitterName=Gizmo_LocalSpace, rendererClasses [&NiagaraMeshRendererProperties]}
  {bEnabled=false, emitterName=FN_Debug, rendererClasses [&NiagaraSpriteRendererProperties]}
 ]
 userVariables [
  User.bIsFrontEnd {type=NiagaraBool}
  User.bIsFrontEndPreview {type=NiagaraBool}
  User.ParaGlideLeanAlpha {type=NiagaraFloat, defaultValue=0, description=""}
 ]
}
